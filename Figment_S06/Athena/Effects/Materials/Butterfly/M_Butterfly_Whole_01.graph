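# Translucent unlit mesh-particle material for a Butterfly effect, producing an animated, noisy rim-band mask.
# A Fresnel driven by the MiddleFade dynamic parameter is multiplied by its inverse and sharpened, isolating a band between center and silhouette.
# Two differently tiled fire-noise samples, scrolled by PanSpeed, are multiplied and boosted to break up that band.
# The clamped result is scaled by particle alpha and a second parameter-driven Fresnel.
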Material /Figment_S06/Athena/Effects/Materials/Butterfly/M_Butterfly_Whole_01
graph {
 properties {bUsedWithMeshParticles=true, blendMode=BLEND_Translucent, pixelDepthOffsetMode=PDOM_Legacy, shadingModel=MSM_Unlit, usedShadingModels=Unlit}
 nodes [
  n1 = DynamicParameter() {defaultValue=<rgba 0 1 1 1>}
  n2 = VertexNormalWS()
  n3 = DynamicParameter() {defaultValue=<rgba 3 1 1 1>, paramNames [MiddleFade,EdgeFade,PanSpeed,Param4]}
  n4 = Fresnel(ExponentIn=@n3.MiddleFade) {baseReflectFraction=0.001}
  n5 = OneMinus(None=@n4)
  n6 = Power(Base=Multiply(A=Multiply(A=@n5, B=@n4), B=Constant() {r=1}))
  n7 = Multiply(A=Time(), B=@n3.PanSpeed)
  n8 = Panner(Coordinate=TextureCoordinate(), Time=@n7) {speedX=1}
  n9 = Panner(Coordinate=TextureCoordinate() {uTiling=4, vTiling=2}, Time=@n7) {speedX=1}
  n10 = Multiply(
   A=Multiply(
    A=TextureSample(UVs=@n8) {texture=/Figment_S06/Effects/Fort_Effects/Textures/Tile/T_FireNoise_01}.G
    B=TextureSample(UVs=@n9) {texture=&T_FireNoise_01}.G
   )
   B=Constant() {r=16}
  )
  Multiply(A=Multiply(A=ParticleColor().A, B=Fresnel(BaseReflectFractionIn=@n1.Param1, Normal=@n2)), B=Clamp(None=Multiply(A=@n6, B=@n10)))
 ]
}
